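# Default Lit material, named as the neon top of the Dragon Building in the NeonCity set and kept in a LightingTemp folder.
# It exposes a Roughness scalar and a Color vector whose overbright cyan-blue default suggests glow, though the graph shows no connections to material outputs.
# Usage flags cover Nanite, instanced static meshes, geometry collections and static lighting.
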Material /Game/Environments/Asteria/Sets/NeonCity/Techart/LightingTemp/DragonBuildingNeonTop
graph {
 properties {
  bUsedWithGeometryCollections=true
  bUsedWithInstancedStaticMeshes=true
  bUsedWithNanite=true
  bUsedWithStaticLighting=true
  pixelDepthOffsetMode=PDOM_Legacy
  usedShadingModels=Default Lit
 }
 nodes [
  ScalarParameter() {defaultValue=0.6407, parameterName=Roughness}
  VectorParameter() {parameterName=Color, defaultValue=<rgba 0.8073 3.1055 10 1>}
 ]
}
